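# MetaSound source preset for the first-person firing sound of the Red Dot assault rifle.
# It inherits its graph from MSS_Base_Weapon_Fire_1p_Simple, supplying six main-fire wave variations, a rifle low-ammo sound, and settings for mix, random pitch/volume, aim-down-sights volume and firing timing.
# It exposes the Fort.Weapon interface as a 5.1 one-shot source, using local gunfire concurrency and first-person weapon-fire volume modulation.
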MetaSoundSource /WeaponsUnvaulted/Assets/Audio/RedDot_AR/MSS_WPN_RedDotAR_1p_Preset
preset_parent=/FortMetaSoundWeapon/Items/Metasounds/MSS_Base_Weapon_Fire_1p_Simple
inputs [
 bHasUnlimitedAmmo {type=Bool}
 bIsFiring {type=Bool}
 BurstRate {type=Time}
 FireBurst {type=Trigger}
 FireSingleShot {type=Trigger}
 LowAmmo.ADSMainVolumeMultiply {type=Float}
 LowAmmo.Random.MaxPitch {type=Float}
 LowAmmo.Random.MaxVolume {type=Float}
 LowAmmo.Random.MinPitch {type=Float}
 LowAmmo.Random.MinVolume {type=Float}
 LowAmmo.Sounds {type=WaveAsset:Array}
 LowAmmo.Threshold {type=Int32}
 MainFire.ADSMainVolumeMultiply {type=Float}
 MainFire.LFVolume {type=Float}
 MainFire.Random.MaxPitch {type=Float}
 MainFire.Random.MaxVolume {type=Float}
 MainFire.Random.MinPitch {type=Float}
 MainFire.Random.MinVolume {type=Float}
 MainFire.Sounds {type=WaveAsset:Array}
 MainFire.StopTriggerDelay {type=Time}
 Mixing.LowAmmoVolume {type=Float}
 Mixing.MainFireVolume {type=Float}
 Mixing.WeaponOutputVolume {type=Float}
 RetriggerDelayTime {type=Time}
 RoundsPerBurst {type=Int32}
 StartFiring {type=Trigger}
 StopBurst {type=Trigger}
 StopFiring {type=Trigger}
 UE.Source.OnPlay {type=Trigger}
 UseSingleVoice {type=Bool}
 Weapon.AmmoRemaining {type=Int32}
 Weapon.ClipSize {type=Int32}
 Weapon.FireRate {type=Float}
 Weapon.FireStartTime {type=Float}
 Weapon.IsAimingDownSights {type=Bool}
 Weapon.IsSuppressorEnabled {type=Bool}
]
interfaces [Fort.Weapon:1.0, UE.OutputFormat.5dot1:1.0, UE.Source.OneShot:1.0, UE.Source:1.1]
outputs [
 bIsAlmostFinished {type=Bool}
 "UE.OutputFormat.5dot1.Audio:0" {type=Audio}
 "UE.OutputFormat.5dot1.Audio:1" {type=Audio}
 "UE.OutputFormat.5dot1.Audio:2" {type=Audio}
 "UE.OutputFormat.5dot1.Audio:3" {type=Audio}
 "UE.OutputFormat.5dot1.Audio:4" {type=Audio}
 "UE.OutputFormat.5dot1.Audio:5" {type=Audio}
 UE.Source.OneShot.OnFinished {type=Trigger}
]
pages [Default {input_defaults {
   BurstRate=0.4
   FireBurst=null
   FireSingleShot=null
   LowAmmo.ADSMainVolumeMultiply=1
   LowAmmo.Random.MaxPitch=1.05
   LowAmmo.Random.MaxVolume=1.05
   LowAmmo.Random.MinPitch=0.95
   LowAmmo.Random.MinVolume=0.95
   LowAmmo.Threshold=10
   MainFire.ADSMainVolumeMultiply=1.2
   MainFire.LFVolume=1
   MainFire.Random.MaxPitch=1.05
   MainFire.Random.MaxVolume=1.05
   MainFire.Random.MinPitch=0.95
   MainFire.Random.MinVolume=0.95
   MainFire.StopTriggerDelay=0.03
   Mixing.LowAmmoVolume=1
   Mixing.MainFireVolume=0.7
   Mixing.WeaponOutputVolume=1
   RetriggerDelayTime=2
   RoundsPerBurst=2
   StartFiring=null
   StopBurst=null
   StopFiring=null
   UE.Source.OnPlay=null
   UseSingleVoice="<MetaSoundBuilderResult.SUCCEEDED: 0>"
   Weapon.AmmoRemaining=30
   Weapon.ClipSize=30
   Weapon.FireRate=0.5
   Weapon.FireStartTime=0
   Weapon.IsAimingDownSights=null
   Weapon.IsSuppressorEnabled=null
   bHasUnlimitedAmmo=null
   bIsFiring=null
   LowAmmo.Sounds [/Game/Sounds/Fort_Weapons/LowAmmo/weapon_low_ammo_hi_rifle_02]
   MainFire.Sounds [
    &/MSS_WPN_Assault_RedDotAR_1P_01
    &/MSS_WPN_Assault_RedDotAR_1P_02
    &/MSS_WPN_Assault_RedDotAR_1P_03
    &/MSS_WPN_Assault_RedDotAR_1P_04
    &/MSS_WPN_Assault_RedDotAR_1P_05
    &/MSS_WPN_Assault_RedDotAR_1P_06
   ]
  }}]
properties {
 assetImportData=&self:AssetImportData
 numChannels=6
 outputFormat=FiveDotOne
 qualitySetting=Fort_Weapons
 sampleRate=48000
 soundClassObject=&Fort_WeaponRanged_Fire_1P_PSM
 concurrencyOverrides {maxCount=2, resolutionRule=StopOldest}
 concurrencySet [&Weapon_Gunfire_Local_Conc]
 modulationSettings {
  volumeRouting=Union
  volumeModulationDestination {bEnableModulation=true, modulators [/FortMetaSoundWeapon/Items/Modulation/PP_Weapon_Fire_1p_Vol]}
 }
}
